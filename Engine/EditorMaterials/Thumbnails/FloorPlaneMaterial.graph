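# Editor material for the floor plane in thumbnails, producing a very dark, low-contrast checkerboard.
# A screen-aligned UV function, fed a 1024 by 1024 texture size, drives the CheckerPattern function with uniform tiling.
# The checker result blends between two near-black greys, so the pattern follows screen space rather than the mesh's UVs.
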Material /Engine/EditorMaterials/Thumbnails/FloorPlaneMaterial
graph {
 properties {bUsedWithStaticLighting=true, pixelDepthOffsetMode=PDOM_Legacy, usedShadingModels=Default Lit}
 nodes [
  Constant3Vector() {constant=<rgba 0 0 0 1>}
  n1 = Constant2Vector() {g=1024, r=1024}
  LinearInterpolate(
   A=Constant3Vector() {constant=<rgba 0.01 0.01 0.01 1>}
   B=Constant3Vector() {constant=<rgba 0.015 0.015 0.015 1>}
   Alpha=MaterialFunctionCall!(
    UVs=MaterialFunctionCall(Texture Dimensions=@n1) {materialFunction=&ScreenAlignedUVs}."X 100%, Y 100%"
    Tiling=Constant2Vector() {g=10, r=10}
   ) {materialFunction=/Engine/ArtTools/RenderToTexture/MaterialFunctions/CheckerPattern}.Result
  )
 ]
}
